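# Sound cue for the finisher swipe of the MistRaven Connector's ground dash attack.
# A Random node picks one of five equally weighted Attk_Swipe_04 wave variations and feeds a Modulator that applies a fixed, slightly lowered pitch.
# It plays through the melee weapon sound class with close-range stereo attenuation, and a concurrency override caps simultaneous instances.
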SoundCue /MistRaven/Animation/ConnectorAnim/Audio/SC_MistRaven_Connector_Ground_Dash_Swipe_Finisher
root=Modulator_0
nodes [
 {id=Modulator_0, children [Random_1], properties {pitchMax=0.9, pitchMin=0.9}}
 {id=Random_1, children [WavePlayer_5,WavePlayer_6,WavePlayer_7,WavePlayer_8,WavePlayer_9], properties {weights [1,1,1,1,1]}}
 {id=WavePlayer_5, properties {soundWaveAssetPtr=/MistRaven/Animation/ConnectorAnim/Audio/wav/MistRaven_Connector_Attk_Swipe_04_01}}
 {id=WavePlayer_6, properties {soundWaveAssetPtr=/MistRaven/Animation/ConnectorAnim/Audio/wav/MistRaven_Connector_Attk_Swipe_04_02}}
 {id=WavePlayer_7, properties {soundWaveAssetPtr=/MistRaven/Animation/ConnectorAnim/Audio/wav/MistRaven_Connector_Attk_Swipe_04_03}}
 {id=WavePlayer_8, properties {soundWaveAssetPtr=/MistRaven/Animation/ConnectorAnim/Audio/wav/MistRaven_Connector_Attk_Swipe_04_04}}
 {id=WavePlayer_9, properties {soundWaveAssetPtr=/MistRaven/Animation/ConnectorAnim/Audio/wav/MistRaven_Connector_Attk_Swipe_04_05}}
]
properties {
 attenuationSettings=/MistRaven/Animation/ConnectorAnim/Audio/Settings/CS_MistRaven_Connector_Melee_Attack_Close_Range_Stereo_Att
 bOverrideConcurrency=true
 maxDistance=7000
 soundClassObject=&Fort_WeaponMelee_PSM
 volumeMultiplier=1.1
 concurrencyOverrides {maxCount=3}
}
